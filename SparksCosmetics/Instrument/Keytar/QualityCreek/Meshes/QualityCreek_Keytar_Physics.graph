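# Physics asset for the QualityCreek keytar instrument cosmetic, defining physics bodies for its dangling headphones and ribbons.
# The handle bone carries two capsules, for neck and body, and anchors the headphones and the ribbon root.
# Two capsule chains, five-link ribbon A and four-link ribbon B, branch from that root.
# Angular limits hold the headphones tightly, while the ribbon links swing and twist more widely.
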PhysicsAsset /SparksCosmetics/Instrument/Keytar/QualityCreek/Meshes/QualityCreek_Keytar_Physics
bodies {
 dyn_headphones [{shapeType=Capsule, length=4.0419, radius=6.0892, shapeName=dyn_headphones_capsule, center=<3 0 0>, rotation=<rot 90 0 0>}]
 dyn_ribbon_a_01 [{shapeType=Capsule, length=1, radius=1, shapeName=dyn_ribbon_a_01_capsule, center=<0.7924 0 0>, rotation=<rot -90 160.5288 -160.5289>}]
 dyn_ribbon_a_02 [{shapeType=Capsule, length=1, radius=1, shapeName=dyn_ribbon_a_02_capsule, center=<0.7924 0 0>, rotation=<rot -90 0.0001 -0.0003>}]
 dyn_ribbon_a_03 [{shapeType=Capsule, length=1, radius=1, shapeName=dyn_ribbon_a_03_capsule, center=<0.7924 0 0>, rotation=<rot -90 179.7468 -179.7469>}]
 dyn_ribbon_a_04 [{shapeType=Capsule, length=1, radius=1, shapeName=dyn_ribbon_a_04_capsule, center=<0.7924 0 0>, rotation=<rot -90 90 -90.0001>}]
 dyn_ribbon_a_05 [{shapeType=Capsule, length=1, radius=1, shapeName=dyn_ribbon_a_05_capsule, center=<0.7924 0 0>, rotation=<rot -90 5.0512 -5.0512>}]
 dyn_ribbon_b_01 [{shapeType=Capsule, length=1, radius=1, shapeName=dyn_ribbon_b_01_capsule, center=<0.7924 0 0>, rotation=<rot -90 179.9999 179.9999>}]
 dyn_ribbon_b_02 [{shapeType=Capsule, length=1, radius=1, shapeName=dyn_ribbon_b_02_capsule, center=<0.7924 0 0>, rotation=<rot -90 0.0001 -0.0004>}]
 dyn_ribbon_b_03 [{shapeType=Capsule, length=1, radius=1, shapeName=dyn_ribbon_b_03_capsule, center=<0.7924 0 0>, rotation=<rot -90 180 179.9998>}]
 dyn_ribbon_b_04 [{shapeType=Capsule, length=1, radius=1, shapeName=dyn_ribbon_b_04_capsule, center=<0.7924 0 0>, rotation=<rot -90 179.9999 179.9998>}]
 dyn_ribbon_root [{shapeType=Capsule, length=1, radius=1, shapeName=dyn_ribbon_root_capsule}]
 handle [
  {shapeType=Capsule, length=28, radius=3, shapeName=handle_capsule_neck, center=<0 1.0092 45.4717>}
  {shapeType=Capsule, length=9, radius=3.5, shapeName=handle_capsule_body, center=<0.2744 -3.9457 35.9026>, rotation=<rot 0.0003 0 -75>}
 ]
}
constraints [
 {bone1Name=dyn_headphones, bone2Name=handle, swing1LimitDegrees=7.5, swing2LimitDegrees=10, twistLimitDegrees=5}
 {bone1Name=dyn_ribbon_root, bone2Name=handle, swing1LimitDegrees=10, swing2LimitDegrees=45, twistLimitDegrees=30}
 {bone1Name=dyn_ribbon_a_01, bone2Name=dyn_ribbon_root, swing1LimitDegrees=30, swing2LimitDegrees=45, twistLimitDegrees=30}
 {bone1Name=dyn_ribbon_a_02, bone2Name=dyn_ribbon_a_01, swing1LimitDegrees=30, swing2LimitDegrees=45, twistLimitDegrees=30}
 {bone1Name=dyn_ribbon_a_03, bone2Name=dyn_ribbon_a_02, swing1LimitDegrees=30, swing2LimitDegrees=45, twistLimitDegrees=30}
 {bone1Name=dyn_ribbon_a_04, bone2Name=dyn_ribbon_a_03, swing1LimitDegrees=30, swing2LimitDegrees=45, twistLimitDegrees=30}
 {bone1Name=dyn_ribbon_a_05, bone2Name=dyn_ribbon_a_04, swing1LimitDegrees=30, swing2LimitDegrees=45, twistLimitDegrees=30}
 {bone1Name=dyn_ribbon_b_01, bone2Name=dyn_ribbon_root, swing1LimitDegrees=30, swing2LimitDegrees=45, twistLimitDegrees=30}
 {bone1Name=dyn_ribbon_b_02, bone2Name=dyn_ribbon_b_01, swing1LimitDegrees=30, swing2LimitDegrees=45, twistLimitDegrees=30}
 {bone1Name=dyn_ribbon_b_03, bone2Name=dyn_ribbon_b_02, swing1LimitDegrees=30, swing2LimitDegrees=45, twistLimitDegrees=30}
 {bone1Name=dyn_ribbon_b_04, bone2Name=dyn_ribbon_b_03, swing1LimitDegrees=30, swing2LimitDegrees=45, twistLimitDegrees=30}
]
